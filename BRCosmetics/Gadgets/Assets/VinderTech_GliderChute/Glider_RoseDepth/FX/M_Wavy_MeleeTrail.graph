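# Translucent, unlit, two-sided wavy trail material for Niagara ribbons, kept in the RoseDepth glider's FX folder.
# Panning noise and a time-driven sine warp the UVs, with the wave ramping in along U and tuned by dynamic parameters for tiling, offset and distortion.
# A high-power sine of the warped V forms a thin streak, blended with its noise-broken version and scaled by particle alpha.
# Particle color feeds an emissive adjustment function.
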Material /BRCosmetics/Gadgets/Assets/VinderTech_GliderChute/Glider_RoseDepth/FX/M_Wavy_MeleeTrail
graph {
 properties {
  bUsedWithNiagaraRibbons=true
  blendMode=BLEND_Translucent
  opacityMaskClipValue=0.1
  pixelDepthOffsetMode=PDOM_Legacy
  shadingModel=MSM_Unlit
  twoSided=true
  usedShadingModels=Unlit
 }
 nodes [
  n1 = ParticleColor()
  MaterialFunctionCall(Emissive=@n1.RGB) {materialFunction=&MF_FX_EmissiveAdjustments}
  n2 = TextureCoordinate() {uTiling=0.4, vTiling=0.4}
  n3 = DynamicParameter() {defaultValue=<rgba 1 1 0.25 1>, paramNames [SineTile, Offset, SineDistortion, Param4]}
  n4 = TextureSample(UVs=Panner(Coordinate=Add(A=@n2, B=@n3.Offset)) {speedX=-1.1, speedY=0.3}) {samplerType=SAMPLERTYPE_Grayscale, texture=/Game/Global/Textures/Default/Noise/LowResBlurredNoiseNormalized}
  n5 = Add(A=Add(A=TextureCoordinate(), B=Constant2Vector() {g=-0.05}), B=Multiply(A=@n4.R) {constB=0.2})
  n6 = Time()
  n7 = Add(A=Multiply(A=ComponentMask(None=@n5) {r=true}, B=@n3.SineTile), B=Multiply(A=Add(A=@n6, B=@n3.Offset)) {constB=-1.8})
  n8 = TextureCoordinate()
  n9 = Saturate(None=Divide(A=ComponentMask(None=@n8) {r=true}) {constB=0.3})
  n10 = Add(A=Multiply(A=Sine(None=@n7) {period=0.67}, B=Multiply(A=@n3.SineDistortion, B=@n9)) {constB=0.25}, B=@n5)
  n11 = Power(Base=Sine(None=ComponentMask(None=@n10) {g=true}) {period=2}) {constExponent=14}
  n12 = Multiply(B=Saturate(None=@n11)) {constA=1}
  n13 = Multiply(A=@n4.B, B=@n12)
  Multiply(A=@n13, B=ScalarParameter() {defaultValue=2, parameterName=Opacity_Mult})
  Multiply(A=@n1.A, B=LinearInterpolate(A=@n13, B=@n12) {constAlpha=0.4})
  Round(None=@n11)
  n14 = TextureCoordinate()
  Saturate(None=Divide(A=ComponentMask(None=@n14) {r=true}) {constB=0.05})
 ]
}
